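# Animation Blueprint for the CupidEvil backpack, adapting its pose and secondary motion to player state.
# Owner spine_05 bend offsets the pack bone, with extra additive offsets for sprinting, diving, floating and parachute.
# State bools switch separately tuned Anim Dynamics stacks on the three-bone Dyn_chain for floating, not floating, DBNO, carried and crouching.
# Events set the Hide Element 01 material parameter while a CupidEvil-tagged pickaxe is equipped outside skydiving.
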AnimBlueprint /BRCosmetics/Accessories/FORT_Backpacks/Backpack_M_MED_CupidEvil/Meshes/M_MED_CupidEvil_Pack_AnimBP
target_skeleton=&/M_MED_CupidEvil_Pack_Skeleton
graphs [
 AnimGraph {nodes [
   n1 = Get bIsDBNOCarried()
   n2 = Get bIsDBNO()
   n3 = Get not_floating() -> (not_floating:bool)
   n4 = "Transform (Modify) Bone - Bone: pack"(ComponentPose=Local To Component(LocalPose=Input Pose()), Translation=Get backpack_translation(), Rotation=Get backpack_rotate()) {
    rotationMode=BMM_Additive
    scaleMode=BMM_Additive
    scaleSpace=BCS_BoneSpace
    translationMode=BMM_Additive
    alphaBoolBlend {blendOption=Linear}
    boneToModify {boneName=pack}
   }
   n5 = Get bIsSprinting()
   n6 = "Transform (Modify) Bone - Bone: Dyn_chain_1"!(ComponentPose=@n4, bAlphaBoolEnabled:bool=@n5, Translation="0,0,0", Rotation="0,-5,0", Scale="1,1,1") -> (Pose) {
    alphaInputType=Bool
    rotationMode=BMM_Additive
    rotationSpace=BCS_BoneSpace
    alphaBoolBlend {blendInTime=0.1, blendOption=ExpOut, blendOutTime=0.2}
    boneToModify {boneName=Dyn_chain_1}
    rotation {yaw=-5}
   }
   n7 = "Transform (Modify) Bone - Bone: Dyn_chain_2"!(ComponentPose=@n6, bAlphaBoolEnabled:bool=@n5, Translation="0,0,0", Rotation="0,-5,0", Scale="1,1,1") -> (Pose) {
    alphaInputType=Bool
    rotationMode=BMM_Additive
    rotationSpace=BCS_BoneSpace
    alphaBoolBlend {blendInTime=0.1, blendOption=ExpOut, blendOutTime=0.2}
    boneToModify {boneName=Dyn_chain_2}
    rotation {yaw=-5}
   }
   n8 = "Transform (Modify) Bone - Bone: Dyn_chain_3"!(ComponentPose=@n7, bAlphaBoolEnabled:bool=@n5, Translation="0,0,0", Rotation="0,-10,0", Scale="1,1,1") -> (Pose) {
    alphaInputType=Bool
    rotationMode=BMM_Additive
    rotationSpace=BCS_BoneSpace
    alphaBoolBlend {blendInTime=0.1, blendOption=Custom, blendOutTime=0.6}
    boneToModify {boneName=Dyn_chain_3}
    rotation {yaw=-10}
   }
   n9 = Get bIsSkydiveDiveMode()
   n10 = "Transform (Modify) Bone - Bone: Dyn_chain_1"(ComponentPose=@n8, bAlphaBoolEnabled=@n9, Translation="0,0,0", Rotation="0,-10,0", Scale="1,1,1") {
    alphaInputType=Bool
    rotationMode=BMM_Additive
    rotationSpace=BCS_BoneSpace
    alphaBoolBlend {blendInTime=0.1, blendOption=ExpOut, blendOutTime=0.2}
    boneToModify {boneName=Dyn_chain_1}
    rotation {yaw=-10}
   }
   n11 = "Transform (Modify) Bone - Bone: Dyn_chain_2"(ComponentPose=@n10, bAlphaBoolEnabled=@n9, Translation="0,0,0", Rotation="0,-10,0", Scale="1,1,1") {
    alphaInputType=Bool
    rotationMode=BMM_Additive
    rotationSpace=BCS_BoneSpace
    alphaBoolBlend {blendInTime=0.1, blendOption=ExpOut, blendOutTime=0.2}
    boneToModify {boneName=Dyn_chain_2}
    rotation {yaw=-10}
   }
   n12 = "Transform (Modify) Bone - Bone: Dyn_chain_3"(ComponentPose=@n11, bAlphaBoolEnabled=@n9, Translation="0,0,0", Rotation="0,-20,0", Scale="1,1,1") {
    alphaInputType=Bool
    rotationMode=BMM_Additive
    rotationSpace=BCS_BoneSpace
    alphaBoolBlend {blendInTime=0.1, blendOption=ExpOut, blendOutTime=0.6}
    boneToModify {boneName=Dyn_chain_3}
    rotation {yaw=-20}
   }
   n13 = "Transform (Modify) Bone - Bone: pack"!(ComponentPose=@n12, bAlphaBoolEnabled:bool=Get bIsSkydiveFloating(), Translation="0,-1,1", Rotation="0,0,5", Scale) -> (Pose) {
    alphaInputType=Bool
    rotationMode=BMM_Additive
    rotationSpace=BCS_BoneSpace
    translationMode=BMM_Additive
    translationSpace=BCS_BoneSpace
    alphaBoolBlend {blendInTime=0.1, blendOption=ExpOut, blendOutTime=0.2}
    boneToModify {boneName=pack}
    rotation {roll=5}
    translation {y=-1, z=1}
   }
   n14 = "Transform (Modify) Bone - Bone: pack"(ComponentPose=@n13, bAlphaBoolEnabled=Get bIsParachuteOpened(), Translation="0,-1,2", Rotation="0,0,5") {
    alphaInputType=Bool
    rotationMode=BMM_Additive
    rotationSpace=BCS_BoneSpace
    translationMode=BMM_Additive
    translationSpace=BCS_BoneSpace
    alphaBoolBlend {blendInTime=0.1, blendOption=ExpOut, blendOutTime=0.2}
    boneToModify {boneName=pack}
    rotation {roll=5}
    translation {y=-1, z=2}
   }
   n15 = Get bIsSkydiveFloating()
   n16 = Anim Dynamics - Bone: Dyn_chain_1!(ComponentPose=@n14, bAlphaBoolEnabled:bool=@n15, GravityOverride:Vector="0,0,-800", ExternalForce="0,0,0") -> (Pose) {
    alphaInputType=Bool
    angularDampingOverride=1
    angularSpringConstant=1000
    bAngularSpring=true
    bOverrideAngularDamping=true
    bUseGravityOverride=true
    simulationSpace=RootRelative
    alphaBoolBlend {blendInTime=0.1, blendOption=Linear, blendOutTime=0.1}
    boundBone {boneName=Dyn_chain_1}
    componentAppliedLinearAccClamp=<3000 3000 3000>
    componentLinearAccScale=<0.1 0.1 0.1>
    componentLinearVelScale=<0.4 0.4 0.4>
    gravityOverride {z=-500}
    physicsBodyDefinitions [{
      collisionType=CustomSphere
      sphereCollisionRadius=2.5
      boundBone {boneName=Dyn_chain_1}
      boxExtents=<9 9 9>
      constraintSetup {angularLimitsMax=<10 10 -20>, angularLimitsMin=<-10 -10 -70>, angularTarget=<0 0 0>}
      localJointOffset=<3.5 0 0>
     }]
   }
   n17 = Anim Dynamics - Bone: Dyn_chain_2!(ComponentPose=@n16, bAlphaBoolEnabled:bool=@n15, GravityOverride:Vector="0,0,-600", ExternalForce="0,0,0") -> (Pose) {
    alphaInputType=Bool
    angularDampingOverride=1
    angularSpringConstant=1000
    bAngularSpring=true
    bOverrideAngularDamping=true
    bUseGravityOverride=true
    simulationSpace=RootRelative
    alphaBoolBlend {blendInTime=0.1, blendOption=Linear, blendOutTime=0.1}
    boundBone {boneName=Dyn_chain_2}
    componentAppliedLinearAccClamp=<3000 3000 3000>
    componentLinearAccScale=<0.1 0.1 0.1>
    componentLinearVelScale=<0.4 0.4 0.4>
    gravityOverride {z=-500}
    physicsBodyDefinitions [{
      collisionType=CustomSphere
      sphereCollisionRadius=2.5
      boundBone {boneName=Dyn_chain_2}
      boxExtents=<9 9 9>
      constraintSetup {angularLimitsMax=<10 10 -15>, angularLimitsMin=<-10 -10 -25>, angularTarget=<0 0 0>}
      localJointOffset=<3.5 0 0>
     }]
   }
   n18 = Anim Dynamics - Bone: Dyn_chain_3!(ComponentPose=@n17, bAlphaBoolEnabled:bool=@n15, GravityOverride:Vector="0,0,-400", ExternalForce="0,0,0") -> (Pose) {
    alphaInputType=Bool
    angularDampingOverride=1
    angularSpringConstant=1000
    bAngularSpring=true
    bOverrideAngularDamping=true
    bUseGravityOverride=true
    simulationSpace=RootRelative
    alphaBoolBlend {blendInTime=0.1, blendOption=Linear, blendOutTime=0.1}
    boundBone {boneName=Dyn_chain_3}
    componentAppliedLinearAccClamp=<3000 3000 3000>
    componentLinearAccScale=<0.1 0.1 0.1>
    componentLinearVelScale=<0.4 0.4 0.4>
    gravityOverride {z=-500}
    physicsBodyDefinitions [{
      collisionType=CustomSphere
      sphereCollisionRadius=2.5
      boundBone {boneName=Dyn_chain_3}
      boxExtents=<9 9 9>
      constraintSetup {angularLimitsMax=<10 15 -10>, angularLimitsMin=<-10 -15 -20>, angularTarget=<0 0 0>}
      localJointOffset=<3.5 0 0>
     }]
    planarLimits [{drivingBone {boneName=root}, planeTransform {location=<0 -5 -35>, rotation=<rot 0 0 -90>}}]
   }
   n19 = Anim Dynamics - Bone: Dyn_chain_1(ComponentPose=@n18, bAlphaBoolEnabled=@n3, GravityOverride="0,0,-980", ExternalForce="0,0,0") {
    alphaInputType=Bool
    angularDampingOverride=0.9999
    angularSpringConstant=1000
    bAngularSpring=true
    bOverrideAngularDamping=true
    bUseGravityOverride=true
    simulationSpace=RootRelative
    alphaBoolBlend {blendInTime=0.1, blendOption=Linear, blendOutTime=0.1}
    boundBone {boneName=Dyn_chain_1}
    componentAppliedLinearAccClamp=<3000 3000 3000>
    componentLinearAccScale=<0.3 0.3 0.3>
    componentLinearVelScale=<0.3 0.3 0.3>
    gravityOverride {z=-500}
    physicsBodyDefinitions [{
      collisionType=CustomSphere
      sphereCollisionRadius=2.5
      boundBone {boneName=Dyn_chain_1}
      boxExtents=<12 12 12>
      constraintSetup {angularLimitsMax=<15 15 4>, angularLimitsMin=<-15 -15 -25>, angularTarget=<0 0 0>}
      localJointOffset=<3 0 0>
     }]
    relativeSpaceBone {boneName=pack}
   }
   n20 = Anim Dynamics - Bone: Dyn_chain_2(ComponentPose=@n19, bAlphaBoolEnabled=@n3, GravityOverride="0,0,-850", ExternalForce="0,0,0") {
    alphaInputType=Bool
    angularDampingOverride=0.9999
    angularSpringConstant=1000
    bAngularSpring=true
    bOverrideAngularDamping=true
    bUseGravityOverride=true
    simulationSpace=RootRelative
    alphaBoolBlend {blendInTime=0.1, blendOption=Linear, blendOutTime=0.1}
    boundBone {boneName=Dyn_chain_2}
    componentAppliedLinearAccClamp=<3000 3000 3000>
    componentLinearAccScale=<0.3 0.3 0.3>
    componentLinearVelScale=<0.3 0.3 0.3>
    gravityOverride {z=-500}
    physicsBodyDefinitions [{
      collisionType=CustomSphere
      sphereCollisionRadius=2.5
      boundBone {boneName=Dyn_chain_2}
      boxExtents=<10 10 10>
      constraintSetup {angularLimitsMax=<15 20 4>, angularLimitsMin=<-15 -20 -30>, angularTarget=<0 0 0>}
      localJointOffset=<3 0 0>
     }]
    relativeSpaceBone {boneName=pack}
   }
   n21 = Anim Dynamics - Bone: Dyn_chain_3(ComponentPose=@n20, bAlphaBoolEnabled=@n3, GravityOverride="0,0,-800", ExternalForce="0,0,0") {
    alphaInputType=Bool
    angularDampingOverride=0.9999
    angularSpringConstant=1000
    bAngularSpring=true
    bOverrideAngularDamping=true
    bUseGravityOverride=true
    bUsePlanarLimit=false
    simulationSpace=RootRelative
    alphaBoolBlend {blendInTime=0.1, blendOption=Linear, blendOutTime=0.1}
    boundBone {boneName=Dyn_chain_3}
    componentAppliedLinearAccClamp=<2000 2000 2000>
    componentLinearAccScale=<0.3 0.3 0.3>
    componentLinearVelScale=<0.4 0.4 0.4>
    gravityOverride {z=-500}
    physicsBodyDefinitions [{
      collisionType=CustomSphere
      sphereCollisionRadius=2.5
      boundBone {boneName=Dyn_chain_3}
      boxExtents=<10 8 12>
      constraintSetup {angularLimitsMax=<15 25 -8>, angularLimitsMin=<-15 -25 -20>, angularTarget=<0 0 0>}
      localJointOffset=<3 0 0>
     }]
    planarLimits [{drivingBone {boneName=root}, planeTransform {location=<0 -5 -35>, rotation=<rot 0 0 -60>}}]
    relativeSpaceBone {boneName=pack}
   }
   n22 = Anim Dynamics - Bone: Dyn_chain_1(ComponentPose=@n21, bAlphaBoolEnabled=@n2, GravityOverride="0,0,-980", ExternalForce="0,0,0") {
    alphaInputType=Bool
    angularDampingOverride=0.9999
    angularSpringConstant=1000
    bAngularSpring=true
    bOverrideAngularDamping=true
    bUseGravityOverride=true
    simulationSpace=RootRelative
    alphaBoolBlend {blendInTime=0.1, blendOption=Linear, blendOutTime=0.1}
    boundBone {boneName=Dyn_chain_1}
    componentAppliedLinearAccClamp=<3000 3000 3000>
    componentLinearAccScale=<0.3 0.3 0.3>
    componentLinearVelScale=<0.3 0.3 0.3>
    gravityOverride {z=-500}
    physicsBodyDefinitions [{
      collisionType=CustomSphere
      sphereCollisionRadius=2.5
      boundBone {boneName=Dyn_chain_1}
      boxExtents=<12 12 12>
      constraintSetup {angularLimitsMax=<15 -30 35>, angularLimitsMin=<-15 -30 -25>, angularTarget=<0 0 0>}
      localJointOffset=<3 0 0>
     }]
    relativeSpaceBone {boneName=pack}
   }
   n23 = Anim Dynamics - Bone: Dyn_chain_2(ComponentPose=@n22, bAlphaBoolEnabled=@n2, GravityOverride="0,0,-850", ExternalForce="0,0,0") {
    alphaInputType=Bool
    angularDampingOverride=0.9999
    angularSpringConstant=1000
    bAngularSpring=true
    bOverrideAngularDamping=true
    bUseGravityOverride=true
    simulationSpace=RootRelative
    alphaBoolBlend {blendInTime=0.1, blendOption=Linear, blendOutTime=0.1}
    boundBone {boneName=Dyn_chain_2}
    componentAppliedLinearAccClamp=<3000 3000 3000>
    componentLinearAccScale=<0.3 0.3 0.3>
    componentLinearVelScale=<0.3 0.3 0.3>
    gravityOverride {z=-500}
    physicsBodyDefinitions [{
      collisionType=CustomSphere
      sphereCollisionRadius=2.5
      boundBone {boneName=Dyn_chain_2}
      boxExtents=<10 10 10>
      constraintSetup {angularLimitsMax=<15 -25 15>, angularLimitsMin=<-15 -25 -30>, angularTarget=<0 0 0>}
      localJointOffset=<3 0 0>
     }]
    relativeSpaceBone {boneName=pack}
   }
   n24 = Anim Dynamics - Bone: Dyn_chain_3(ComponentPose=@n23, bAlphaBoolEnabled=@n2, GravityOverride="0,0,-800", ExternalForce="0,0,0") {
    alphaInputType=Bool
    angularDampingOverride=0.9999
    angularSpringConstant=1000
    bAngularSpring=true
    bOverrideAngularDamping=true
    bUseGravityOverride=true
    simulationSpace=RootRelative
    alphaBoolBlend {blendInTime=0.1, blendOption=Linear, blendOutTime=0.1}
    boundBone {boneName=Dyn_chain_3}
    componentAppliedLinearAccClamp=<2000 2000 2000>
    componentLinearAccScale=<0.3 0.3 0.3>
    componentLinearVelScale=<0.4 0.4 0.4>
    gravityOverride {z=-500}
    physicsBodyDefinitions [{
      collisionType=CustomSphere
      sphereCollisionRadius=2.5
      boundBone {boneName=Dyn_chain_3}
      boxExtents=<10 8 12>
      constraintSetup {angularLimitsMax=<15 -20 -5>, angularLimitsMin=<-15 -20 -25>, angularTarget=<0 0 0>}
      localJointOffset=<3 0 0>
     }]
    planarLimits [{drivingBone {boneName=root}, planeTransform {location=<0 -5 -35>, rotation=<rot 0 0 -60>}}]
    relativeSpaceBone {boneName=pack}
   }
   n25 = Anim Dynamics - Bone: Dyn_chain_1(ComponentPose=@n24, bAlphaBoolEnabled=@n1, GravityOverride="0,0,-980", ExternalForce="0,0,0") {
    alphaInputType=Bool
    angularDampingOverride=0.9999
    angularSpringConstant=1000
    bAngularSpring=true
    bOverrideAngularDamping=true
    bUseGravityOverride=true
    simulationSpace=RootRelative
    alphaBoolBlend {blendInTime=0.1, blendOption=Linear, blendOutTime=0.1}
    boundBone {boneName=Dyn_chain_1}
    componentAppliedLinearAccClamp=<3000 3000 3000>
    componentLinearAccScale=<0.3 0.3 0.3>
    componentLinearVelScale=<0.3 0.3 0.3>
    gravityOverride {z=-500}
    physicsBodyDefinitions [{
      collisionType=CustomSphere
      sphereCollisionRadius=2.5
      boundBone {boneName=Dyn_chain_1}
      boxExtents=<12 12 12>
      constraintSetup {angularLimitsMax=<15 15 4>, angularLimitsMin=<-15 -15 -25>, angularTarget=<0 0 0>}
      localJointOffset=<3 0 0>
     }]
    relativeSpaceBone {boneName=pack}
   }
   n26 = Anim Dynamics - Bone: Dyn_chain_2(ComponentPose=@n25, bAlphaBoolEnabled=@n1, GravityOverride="0,0,-850", ExternalForce="0,0,0") {
    alphaInputType=Bool
    angularDampingOverride=0.9999
    angularSpringConstant=1000
    bAngularSpring=true
    bOverrideAngularDamping=true
    bUseGravityOverride=true
    simulationSpace=RootRelative
    alphaBoolBlend {blendInTime=0.1, blendOption=Linear, blendOutTime=0.1}
    boundBone {boneName=Dyn_chain_2}
    componentAppliedLinearAccClamp=<3000 3000 3000>
    componentLinearAccScale=<0.3 0.3 0.3>
    componentLinearVelScale=<0.3 0.3 0.3>
    gravityOverride {z=-500}
    physicsBodyDefinitions [{
      collisionType=CustomSphere
      sphereCollisionRadius=2.5
      boundBone {boneName=Dyn_chain_2}
      boxExtents=<10 10 10>
      constraintSetup {angularLimitsMax=<15 20 4>, angularLimitsMin=<-15 -20 -30>, angularTarget=<0 0 0>}
      localJointOffset=<3 0 0>
     }]
    relativeSpaceBone {boneName=pack}
   }
   n27 = Anim Dynamics - Bone: Dyn_chain_3(ComponentPose=@n26, bAlphaBoolEnabled=@n1, GravityOverride="0,0,-800", ExternalForce="0,0,0") {
    alphaInputType=Bool
    angularDampingOverride=0.9999
    angularSpringConstant=1000
    bAngularSpring=true
    bOverrideAngularDamping=true
    bUseGravityOverride=true
    simulationSpace=RootRelative
    alphaBoolBlend {blendInTime=0.1, blendOption=Linear, blendOutTime=0.1}
    boundBone {boneName=Dyn_chain_3}
    componentAppliedLinearAccClamp=<2000 2000 2000>
    componentLinearAccScale=<0.3 0.3 0.3>
    componentLinearVelScale=<0.4 0.4 0.4>
    gravityOverride {z=-500}
    physicsBodyDefinitions [{
      collisionType=CustomSphere
      sphereCollisionRadius=2.5
      boundBone {boneName=Dyn_chain_3}
      boxExtents=<10 8 12>
      constraintSetup {angularLimitsMax=<15 20 3>, angularLimitsMin=<-15 -20 -20>, angularTarget=<0 0 0>}
      localJointOffset=<3 0 0>
     }]
    planarLimits [{drivingBone {boneName=root}, planeTransform {location=<0 -5 -35>, rotation=<rot 0 0 -60>}}]
    relativeSpaceBone {boneName=pack}
   }
   n28 = Get bIsCrouching()
   n29 = Anim Dynamics - Bone: Dyn_chain_1(ComponentPose=@n27, bAlphaBoolEnabled=@n28, GravityOverride="0,0,-980", ExternalForce="0,0,0") {
    alphaInputType=Bool
    angularDampingOverride=0.9999
    angularSpringConstant=1000
    bAngularSpring=true
    bOverrideAngularDamping=true
    bUseGravityOverride=true
    simulationSpace=RootRelative
    alphaBoolBlend {blendInTime=0.1, blendOption=Linear, blendOutTime=0.1}
    boundBone {boneName=Dyn_chain_1}
    componentAppliedLinearAccClamp=<3000 3000 3000>
    componentLinearAccScale=<0.3 0.3 0.3>
    componentLinearVelScale=<0.3 0.3 0.3>
    gravityOverride {z=-500}
    physicsBodyDefinitions [{
      collisionType=CustomSphere
      sphereCollisionRadius=2.5
      boundBone {boneName=Dyn_chain_1}
      boxExtents=<12 12 12>
      constraintSetup {angularLimitsMax=<15 25 5>, angularLimitsMin=<-15 -25 -20>, angularTarget=<0 0 0>}
      localJointOffset=<3 0 0>
     }]
    relativeSpaceBone {boneName=pack}
   }
   n30 = Anim Dynamics - Bone: Dyn_chain_2(ComponentPose=@n29, bAlphaBoolEnabled=@n28, GravityOverride="0,0,-850", ExternalForce="0,0,0") {
    alphaInputType=Bool
    angularDampingOverride=0.9999
    angularSpringConstant=1000
    bAngularSpring=true
    bOverrideAngularDamping=true
    bUseGravityOverride=true
    simulationSpace=RootRelative
    alphaBoolBlend {blendInTime=0.1, blendOption=Linear, blendOutTime=0.1}
    boundBone {boneName=Dyn_chain_2}
    componentAppliedLinearAccClamp=<3000 3000 3000>
    componentLinearAccScale=<0.3 0.3 0.3>
    componentLinearVelScale=<0.3 0.3 0.3>
    gravityOverride {z=-500}
    physicsBodyDefinitions [{
      collisionType=CustomSphere
      sphereCollisionRadius=2.5
      boundBone {boneName=Dyn_chain_2}
      boxExtents=<10 10 10>
      constraintSetup {angularLimitsMax=<15 35 5>, angularLimitsMin=<-15 -35 -25>, angularTarget=<0 0 0>}
      localJointOffset=<2 0 0>
     }]
    relativeSpaceBone {boneName=pack}
   }
   n31 = Anim Dynamics - Bone: Dyn_chain_3(ComponentPose=@n30, bAlphaBoolEnabled=@n28, GravityOverride="0,0,-800", ExternalForce="0,0,0") {
    alphaInputType=Bool
    angularDampingOverride=0.9999
    angularSpringConstant=1000
    bAngularSpring=true
    bOverrideAngularDamping=true
    bUseGravityOverride=true
    bUsePlanarLimit=false
    simulationSpace=RootRelative
    alphaBoolBlend {blendInTime=0.1, blendOption=Linear, blendOutTime=0.1}
    boundBone {boneName=Dyn_chain_3}
    componentAppliedLinearAccClamp=<3000 3000 3000>
    componentLinearAccScale=<0.3 0.3 0.3>
    componentLinearVelScale=<0.3 0.3 0.3>
    gravityOverride {z=-500}
    physicsBodyDefinitions [{
      collisionType=CustomSphere
      sphereCollisionRadius=2.5
      boundBone {boneName=Dyn_chain_3}
      boxExtents=<10 8 12>
      constraintSetup {angularLimitsMax=<15 40 -8>, angularLimitsMin=<-15 -40 -20>, angularTarget=<0 0 0>}
      localJointOffset=<2 0 0>
     }]
    planarLimits [{drivingBone {boneName=root}, planeTransform {location=<0 -5 -35>, rotation=<rot 0 0 -60>}}]
    relativeSpaceBone {boneName=pack}
   }
   Output Pose(Result=Component To Local(ComponentPose=@n31)) {layerGroup=None}
  ]}
 EventGraph {nodes [
   Event BlueprintUpdateAnimation()
   > n1 = Sequence!(execute) -> (then_0, then_1, then_2:Exec)
   n2 = GetOwnerSkeletalMeshComponent()
   n3 = BreakRotator(InRot=BreakTransform(InTransform=GetDeltaTransformFromRefPose(self=@n2, BoneName=spine_05, BaseName=spine_04)).Rotation)
   > Set backpack_rotate(backpack_rotate=MakeRotator(Roll=MapRangeClamped(Value=@n3.Yaw, InRangeA=0, InRangeB=-30, OutRangeB=-25)))
   > Set backpack_translation(backpack_translation=MakeVector(Y=MapRangeClamped(Value=@n3.Yaw, InRangeA=0, InRangeB=-30, OutRangeB=-10)))
   Set not_floating(execute=@n1.then_2, not_floating:bool=NOT Boolean(A=Get bIsSkydiveFloating())) -> (Output_Get:bool)
   n4 = Get Dynamic Material Instance(execute=[@n8.then, @n1.then_1], SkeletalMeshComponent=GetPartSkeletalMeshComponent())
   n5 = PickaxeItemDefinitionHasAnyMetaTag(Tags=Make Array("[0]"=(TagName="Cosmetics.MetaTags.CupidEvil")), bRequiresEquipped=true)
   > n6 = Branch(Condition=@n5)
   > n7 = Branch(Condition=Get bIsSkyDiving())
   SetScalarParameterValue(execute=[@n9.then, @n10.else, @n7.then, @n6.else], self=@n4, ParameterName=Hide Element 01, Value=0)
   SetScalarParameterValue(execute=[@n10.then, @n7.else], self=@n4, ParameterName=Hide Element 01, Value=1)
   n8 = Event OnWeaponChanged()
   n9 = Event OnBeginSkydiving()
   Event OnEndSkydiving()
   > n10 = Branch(Condition=@n5)
   TryGetPawnOwner()
  ]}
]
variables [
 backpack_rotate {type=Rotator}
 backpack_translation {type=Vector}
 not_floating {type=boolean}
]
